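# Translucent liquid material for Niagara mesh particles, apparently the poison effect on the Freak pickaxe.
# Color blends two Fresnel-driven tints with a matcap-style reflection, sampled by noise-perturbed view-space normals, then tinted by time of day.
# The same panning distortion texture, randomized per particle, modulates opacity along with particle alpha and quality-switched depth fade.
# Dynamic parameters drive a UV-masked pendulum rotation about the mesh origin, with PreviousFrameSwitch supplying the prior position.
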Material /BRCosmetics/Athena/Effects/Materials/Liquid/M_Pickaxe_Freak_Poison_M_Liquid
graph {
 properties {bUsedWithNiagaraMeshParticles=true, blendMode=BLEND_Translucent, pixelDepthOffsetMode=PDOM_Legacy, usedShadingModels=Default Lit}
 nodes [
  n1 = DynamicParameter() {parameterIndex=1, defaultValue=<rgba 0 0 -1 1>, paramNames [RestNormalX, RestNormalY, RestNormalZ, --------------]}
  n2 = Multiply(A=MaterialFunctionCall(R=@n1.RestNormalX, G=@n1.RestNormalX, B=@n1.RestNormalX) {materialFunction=&AppendMany}.RGB) {constB=-1}
  n3 = DynamicParameter() {parameterIndex=3, defaultValue=<rgba 1 1 1 0.15>, paramNames [PosX,PosY,PosZ,Rigidity]}
  n4 = DynamicParameter() {parameterIndex=2, paramNames [MeshOriginX,MeshOriginY,MeshOriginZ,MaskFalloff]}
  n5 = MaterialFunctionCall(R=@n4.MeshOriginX, G=@n4.MeshOriginX, B=@n4.MeshOriginX, A=@n4.MeshOriginX) {materialFunction=&AppendMany}
  n6 = Subtract(
   A=MaterialFunctionCall(R=@n3.PosX, G=@n3.PosX, B=@n3.PosX) {materialFunction=&AppendMany}.RGB
   B=@n5.RGB
  )
  n7 = WorldPosition()
  n8 = RotateAboutAxis(NormalizedRotationAxis=CrossProduct(A=@n2, B=Normalize(VectorInput=@n6)), RotationAngle=@n3.Rigidity, PivotPoint=@n5.RGB, Position=@n7.XYZ)
  n9 = Subtract(A=ComponentMask(None=TextureCoordinate()) {g=true}, B=ScalarParameter() {parameterName=Mask Offset})
  n10 = Saturate(None=Power(Base=Saturate(None=@n9), Exp=@n4.MaskFalloff))
  n11 = DynamicParameter() {paramNames [PrevPosX,PrevPosY,PrevPosZ,GlowAmt]}
  n12 = Subtract(
   A=MaterialFunctionCall(R=@n11.PrevPosX, G=@n11.PrevPosX, B=@n11.PrevPosX) {materialFunction=&AppendMany}.RGB
   B=@n5.RGB
  )
  n13 = RotateAboutAxis(NormalizedRotationAxis=CrossProduct(A=@n2, B=Normalize(VectorInput=@n12)), RotationAngle=@n3.Rigidity, PivotPoint=@n5.RGB, Position=@n7.XYZ)
  LinearInterpolate(
   A=PreviousFrameSwitch()
   B=PreviousFrameSwitch(Current Frame=LinearInterpolate(A=@n8, Alpha=@n10) {constB=0}, Previous Frame=LinearInterpolate(A=@n13, Alpha=@n10) {constB=0})
   Alpha=ScalarParameter() {defaultValue=1, parameterName=PendulumSetup, sliderMax=1}
  ) {constB=0}
  n14 = LinearInterpolate(
   A=VectorParameter() {parameterName=ColorA, defaultValue=<rgba 0 0.6854 0.75 1>}.RGB
   B=VectorParameter() {parameterName=ColorB, defaultValue=<rgba 0 0.1828 0.2 1>}.RGB
   Alpha=MaterialFunctionCall(
    Brightness=ScalarParameter() {defaultValue=1.3, parameterName=FresnelMult}
    Exponent=ScalarParameter() {defaultValue=2, parameterName=FresnelExp}
   ) {materialFunction=&MF_Fresnel_FX}.Fresnel
  )
  n15 = Multiply(
   A=MaterialFunctionCall(
    Brightness=ScalarParameter() {defaultValue=1.3, parameterName=FresnelBMult}
    Exponent=ScalarParameter() {defaultValue=8, parameterName=FresnelBExp}
   ) {materialFunction=&MF_Fresnel_FX}.Fresnel Inverted
   B=@n14
  )
  n16 = TextureCoordinate()
  n17 = ScalarParameter() {defaultValue=0.6, parameterName=DistTile}
  n18 = Time() {bOverride_Period=true, period=64}
  n19 = DynamicParameter() {paramNames [rand 0-1, WPO_Amnt, Param3, Param4]}
  n20 = ScalarParameter() {defaultValue=2, parameterName=DistSpeed}
  n21 = Constant() {r=0.1}
  n22 = TextureSample(UVs=Panner(Coordinate=Multiply(A=@n16, B=@n17), Time=Add(A=@n18, B=@n19.rand 0-1), Speed=AppendVector(A=@n20, B=@n21)) {speedX=-1, speedY=0.1}) {samplerType=SAMPLERTYPE_Normal, texture=&T_Weap_Distort_N_02}
  n23 = Add(A=Multiply(A=Transform(None=@n22.RGB), B=ScalarParameter() {defaultValue=0.5, parameterName=NoiseSTr}), B=VertexNormalWS())
  n24 = Add(A=Multiply(A=Transform(None=@n23) {transformSourceType=TRANSFORMSOURCE_World, transformType=TRANSFORM_View}) {constB=0.5}) {constB=0.5}
  n25 = MaterialFunctionCall(
   In=TextureSample(UVs=ComponentMask(None=@n24) {g=true, r=true}) {texture=/Game/Packages/Fortress_SharedMaterials/Textures/T_Ravens_SpecularReflection}.G
   Contrast=ScalarParameter() {parameterName=ReflectionContrast}
  ) {materialFunction=&CheapContrast}
  n26 = VectorParameter() {parameterName=ReflectionColor, defaultValue=<rgba 0.4 2 2 1>}
  Multiply(
   A=MaterialFunctionCall(
    Color_Morning=VectorParameter() {group=TOD, parameterName=TOD_Morning, defaultValue=<rgba 1 1 1 1>}.RGB
    Color_Day=VectorParameter() {group=TOD, parameterName=TOD_Afternoon, defaultValue=<rgba 1 1 1 1>}.RGB
    Color_Evening=VectorParameter() {group=TOD, parameterName=TOD_Evening, defaultValue=<rgba 1 1 1 1>}.RGB
    Color_Night=VectorParameter() {group=TOD, parameterName=TOD_Night, defaultValue=<rgba 1 1 1 1>}.RGB
   ) {materialFunction=&Color_Time_of_Day}.Result
   B=Add(A=Add(A=@n15, B=@n14), B=Multiply(A=@n25.Result, B=@n26.RGB))
  )
  n27 = Add(A=Multiply(A=@n22.G) {constB=0.5}) {constB=0.5}
  n28 = ScalarParameter() {defaultValue=1, parameterName=contrast}
  n29 = Saturate(None=Add(
    A=ScalarParameter() {defaultValue=0.75, parameterName=OpacBase}
    B=MaterialFunctionCall(In=@n27, Contrast=@n28) {materialFunction=&CheapContrast}.Result
   ))
  n30 = ScalarParameter() {defaultValue=0.95, parameterName=OpacEnd}
  n31 = Constant() {r=1}
  n32 = ShadingPathSwitch(Default=DepthFade(FadeDistance=ScalarParameter() {defaultValue=10, parameterName=DFDist}), Mobile=@n31) {desc=Migrated from FeatureLevelSwitch}
  Multiply(A=Multiply(A=Multiply(A=@n29, B=@n30), B=ParticleColor().A), B=QualitySwitch(Default=@n32, Low=@n31, High=@n32, Medium=@n32))
 ]
}
